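# Animation Blueprint for the quadruped ShirtTilapia mesh in CosmeticCompanions, which drives its skeleton with retargeted animation.
# The AnimGraph is a single Retarget Pose From Mesh node using the RTG_ShirtTilapia IK Retargeter, fed straight into the output pose.
# It copies and retargets the pose of a source skeletal mesh at runtime, so it has no state machine, variables or event logic of its own.
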AnimBlueprint /CosmeticCompanions/Assets/Quadruped/ShirtTilapia/Meshes/ShirtTilapia_Quadruped_AnimBP
target_skeleton=&/ShirtTilapia_Quadruped_Skeleton
graphs [
 AnimGraph {nodes [Output Pose(Result=Retarget Pose From Mesh!() -> (Pose:Pose Link Structure) {is=title AnimGraphNode_RetargetPoseFromMesh, iKRetargeterAsset=&/RTG_ShirtTilapia})]}
 EventGraph {links [], nodes []}
]
variables []
