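# Post-process animation blueprint for the ShirtTilapia quadruped cosmetic companion, layering facial poses and a Control Rig pass onto the incoming pose.
# The input pose drives a face-poses Pose Asset node, then passes through the DefaultSlot montage slot.
# The CRDyn Control Rig runs last, presumably for secondary dynamics, and feeds the output pose.
# The event graph and variables are empty, so all work happens in the AnimGraph.
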AnimBlueprint /CosmeticCompanions/Assets/Quadruped/ShirtTilapia/Meshes/ShirtTilapia_Quadruped_AnimBP_Post
target_skeleton=&/ShirtTilapia_Quadruped_Skeleton
graphs [
 AnimGraph {nodes [
   n1 = ShirtTilapia_Quadruped_FacePoses_PoseAsset!(SourcePose:Pose Link Structure=Input Pose()) -> (Pose:Pose Link Structure) {
    is=title AnimGraphNode_PoseBlendNode
    blendOption=Linear
    poseAsset=&/ShirtTilapia_Quadruped_FacePoses_PoseAsset
   }
   Output Pose(Result=Control Rig(Source=Slot 'DefaultSlot'(Source=@n1)) {controlRigAssetReference {controlRigAsset=&/ShirtTilapia_Quadruped_CRDyn}})
  ]}
 EventGraph {links [], nodes []}
]
variables []
