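# Collision physics asset for the ShirtTilapia quadruped cosmetic companion, apparently for dynamics simulation given its CRDyn name.
# Single capsules approximate the torso and upper limbs: hips, three spine bones, lower neck, front scapulae and mirrored back thighs.
# It defines no constraints, so the bodies serve as bone-following collision volumes rather than a jointed ragdoll.
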PhysicsAsset /CosmeticCompanions/Assets/Quadruped/ShirtTilapia/Meshes/ShirtTilapia_Quadruped_CRDyn_Collision
bodies {
 hips [{shapeType=Capsule, length=5.8049, radius=9.8364, center=<-3.1655 1.9369 0>, rotation=<rot 90 164.9002 0>}]
 neck_01 [{shapeType=Capsule, length=6.3715, radius=7.1876, center=<6.2697 1.352 0>, rotation=<rot 90 -17.9102 0>}]
 scapula_fr_l [{shapeType=Capsule, length=4, radius=6.7, center=<5.8302 1.8603 3.6407>, rotation=<rot 79.8798 -0.0001 0>}]
 scapula_fr_r [{shapeType=Capsule, length=4, radius=6.7, center=<-5.8302 -1.8603 -3.6407>, rotation=<rot 79.8798 -0.0001 -180>}]
 spine_01 [{shapeType=Capsule, length=16.4516, radius=9.447, center=<5.989 0.1841 -0.0097>, rotation=<rot 90 -4.4097 0>}]
 spine_02 [{shapeType=Capsule, length=9.5434, radius=11.8, center=<7.8789 -0.3652 0>, rotation=<rot 90 -9.256 0>}]
 spine_03 [{shapeType=Capsule, length=3.1734, radius=8.8572, center=<-0.2896 -3.1302 0>, rotation=<rot 90 -35.4754 0>}]
 thigh_bk_l [{shapeType=Capsule, length=20.7, radius=9.5, center=<11.5781 -0.9524 -3.3004>, rotation=<rot 83.6128 175.2162 180>}]
 thigh_bk_r [{shapeType=Capsule, length=20.7, radius=9.5, center=<-11.5781 0.9524 3.3004>, rotation=<rot 83.6128 175.2162 0>}]
}
constraints []
