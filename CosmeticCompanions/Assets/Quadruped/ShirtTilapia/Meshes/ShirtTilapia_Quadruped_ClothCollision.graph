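# Physics asset for cloth collision on the ShirtTilapia quadruped cosmetic companion mesh.
# It defines bodies on the hips, neck, both front shoulders, tail base and three middle dynamic cape target bones.
# It has no constraints, so the bodies are not jointed into a ragdoll.
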PhysicsAsset /CosmeticCompanions/Assets/Quadruped/ShirtTilapia/Meshes/ShirtTilapia_Quadruped_ClothCollision
bodies {dyn_cape_mid_02_tgt [], dyn_cape_mid_03_tgt [], dyn_cape_mid_04_tgt [], hips [], neck_01 [], shoulder_fr_l [], shoulder_fr_r [], tail_01 []}
constraints []
